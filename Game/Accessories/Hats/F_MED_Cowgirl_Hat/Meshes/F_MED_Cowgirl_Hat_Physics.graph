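# Physics asset for the F_MED_Cowgirl_Hat accessory mesh, covering the dynamic hair bones that hang from the head.
# Six chains of box and capsule bodies (back, back-left, back-right, left, right, front-right) each link two or three bones to the head, with tight twist limits and wider swing limits.
# Head, neck, upper spine and clavicle shapes approximate the upper body, joined by constraints with free swing and twist.
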PhysicsAsset /Game/Accessories/Hats/F_MED_Cowgirl_Hat/Meshes/F_MED_Cowgirl_Hat_Physics
bodies {
 clavicle_l [{shapeType=Capsule, length=13, radius=5.5, shapeName=clavicle_l_capsule, center=<3.5 -1.5 -0.5>, rotation=<rot 80 -0.0003 -0.0003>}]
 clavicle_r [{shapeType=Capsule, length=13, radius=5.5, shapeName=clavicle_r_capsule, center=<-3.5 1.5 0.5>, rotation=<rot 80 -0.0003 -0.0004>}]
 dyn_hair_bk_01 [{shapeType=Box, extentX=5.9004, extentY=1.8791, extentZ=13.0124, center=<3.2348 0 -0.0001>}]
 dyn_hair_bk_02 [{shapeType=Box, extentX=6.0425, extentY=1.8791, extentZ=14.0124, center=<3.5493 0 -0.0001>}]
 dyn_hair_bk_03 [{shapeType=Box, extentX=9.4028, extentY=2.5131, extentZ=14.5124, center=<4.9698 0 0.0001>}]
 dyn_hair_bk_l_01 [{shapeType=Box, extentX=5.9004, extentY=1.8791, extentZ=7.5124, center=<3.2348 0 -0.0001>}]
 dyn_hair_bk_l_02 [{shapeType=Box, extentX=5.9004, extentY=1.8791, extentZ=7.5124, center=<3.2348 0 -0.0001>}]
 dyn_hair_bk_l_03 [{shapeType=Box, extentX=3.5258, extentY=1.8791, extentZ=6.9635, center=<2.072 0 -0.0001>}]
 dyn_hair_bk_r_01 [{shapeType=Capsule, length=9.235, radius=1.3123, center=<3.2212 0 0>}]
 dyn_hair_bk_r_02 [{shapeType=Capsule, length=9.235, radius=2, center=<2.7466 -0.5 0>}]
 dyn_hair_bk_r_03 [{shapeType=Capsule, length=8, radius=3, center=<3 -1 0>}]
 dyn_hair_fr_r_01 [{shapeType=Box, extentX=5.9004, extentY=1.8791, extentZ=7.5124, center=<3.2348 0 -0.0001>}]
 dyn_hair_fr_r_02 [{shapeType=Box, extentX=5.9004, extentY=1.8791, extentZ=7.5124, center=<3.2348 0 -0.0001>}]
 dyn_hair_l_01 [{shapeType=Box, extentX=5.9004, extentY=1.6203, extentZ=7.5124, center=<3.2348 0 -0.0001>}]
 dyn_hair_l_02 [{shapeType=Box, extentX=5.2096, extentY=1.8791, extentZ=7.5124, center=<3.2348 0 -0.0001>}]
 dyn_hair_l_03 [{shapeType=Box, extentX=3.2866, extentY=1.8791, extentZ=7.5124, center=<2.0441 0 -0.0001>}]
 dyn_hair_r_01 [{shapeType=Box, extentX=5.9004, extentY=1.8791, extentZ=7.5124, center=<3.2348 0 -0.0001>}]
 dyn_hair_r_02 [{shapeType=Capsule, length=28.5882, radius=3, center=<3 1 0>, rotation=<rot 0 0 -5>}]
 dyn_hair_r_03 [{shapeType=Capsule, length=23.3177, radius=2.5, center=<2.8704 0 0>}]
 head [
  {shapeType=Sphere, radius=6.6677, center=<1.3281 -0.8773 0>}
  {shapeType=Capsule, length=5.9657, radius=5.522, center=<0.8117 2.2232 0>, rotation=<rot 90 0 0>}
 ]
 neck_01 [{shapeType=Capsule, length=6.0302, radius=7.613, rotation=<rot 90 0 0>}]
 spine_05 [
  {shapeType=Capsule, length=8, radius=11, shapeName=spine_05_capsule, center=<-3 -1 0>}
  {shapeType=Capsule, length=8, radius=8, shapeName=spine_05_capsule, center=<-7 6 0>}
 ]
}
constraints [
 {bone1Name=dyn_hair_bk_l_01, bone2Name=head, swing1LimitDegrees=45, swing2LimitDegrees=45, twistLimitDegrees=5}
 {bone1Name=dyn_hair_bk_l_02, bone2Name=dyn_hair_bk_l_01, swing1LimitDegrees=45, swing2LimitDegrees=45, twistLimitDegrees=5}
 {bone1Name=dyn_hair_bk_l_03, bone2Name=dyn_hair_bk_l_02, swing1LimitDegrees=45, swing2LimitDegrees=45, twistLimitDegrees=5}
 {bone1Name=dyn_hair_bk_01, bone2Name=head, swing1LimitDegrees=45, swing2LimitDegrees=45, twistLimitDegrees=5}
 {bone1Name=dyn_hair_bk_02, bone2Name=dyn_hair_bk_01, swing1LimitDegrees=60, swing2LimitDegrees=60, twistLimitDegrees=5}
 {bone1Name=dyn_hair_bk_03, bone2Name=dyn_hair_bk_02, swing1LimitDegrees=60, swing2LimitDegrees=60, twistLimitDegrees=5}
 {bone1Name=dyn_hair_l_01, bone2Name=head, swing1LimitDegrees=45, swing2LimitDegrees=45, twistLimitDegrees=5}
 {bone1Name=dyn_hair_l_02, bone2Name=dyn_hair_l_01, swing1LimitDegrees=45, swing2LimitDegrees=45, twistLimitDegrees=5}
 {bone1Name=dyn_hair_l_03, bone2Name=dyn_hair_l_02, swing1LimitDegrees=45, swing2LimitDegrees=45, twistLimitDegrees=5}
 {bone1Name=dyn_hair_bk_r_01, bone2Name=head, swing1LimitDegrees=60, swing2LimitDegrees=45, twistLimitDegrees=5}
 {bone1Name=dyn_hair_bk_r_02, bone2Name=dyn_hair_bk_r_01, swing1LimitDegrees=60, swing2LimitDegrees=45, twistLimitDegrees=5}
 {bone1Name=dyn_hair_bk_r_03, bone2Name=dyn_hair_bk_r_02, swing1LimitDegrees=60, swing2LimitDegrees=45, twistLimitDegrees=5}
 {bone1Name=dyn_hair_r_01, bone2Name=head, swing1LimitDegrees=45, swing2LimitDegrees=45, twistLimitDegrees=5}
 {bone1Name=dyn_hair_r_02, bone2Name=dyn_hair_r_01, swing1LimitDegrees=45, swing2LimitDegrees=45, twistLimitDegrees=5}
 {bone1Name=dyn_hair_r_03, bone2Name=dyn_hair_r_02, swing1LimitDegrees=45, swing2LimitDegrees=45, twistLimitDegrees=5}
 {bone1Name=dyn_hair_fr_r_01, bone2Name=head, swing1LimitDegrees=45, swing2LimitDegrees=45, twistLimitDegrees=5}
 {bone1Name=dyn_hair_fr_r_02, bone2Name=dyn_hair_fr_r_01, swing1LimitDegrees=45, swing2LimitDegrees=45, twistLimitDegrees=5}
 {
  bone1Name=clavicle_l
  bone2Name=spine_05
  swing1LimitDegrees=45
  swing1Motion=Free
  swing2LimitDegrees=45
  swing2Motion=Free
  twistLimitDegrees=45
  twistMotion=Free
 }
 {
  bone1Name=clavicle_r
  bone2Name=spine_05
  swing1LimitDegrees=45
  swing1Motion=Free
  swing2LimitDegrees=45
  swing2Motion=Free
  twistLimitDegrees=45
  twistMotion=Free
 }
 {
  bone1Name=head
  bone2Name=neck_01
  swing1LimitDegrees=45
  swing1Motion=Free
  swing2LimitDegrees=45
  swing2Motion=Free
  twistLimitDegrees=45
  twistMotion=Free
 }
 {
  bone1Name=neck_01
  bone2Name=spine_05
  swing1LimitDegrees=45
  swing1Motion=Free
  swing2LimitDegrees=45
  swing2Motion=Free
  twistLimitDegrees=45
  twistMotion=Free
 }
]
